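; Season 12 frontend elevator outer-door indicator that displays a state on five floor-entry slots.
; The construction script creates a material instance per slot; SetEntry functions color each from a state array and show a bang for state 5.
; SetHenchmenState derives a state code from item ownership and A/B, Ego and Alter quest progress.
; CameraChanged calls RoomOwnerChanged on the main-room camera, but that handler and BeginPlay are empty here.
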(event EventBeginPlay)

(event Custom|RoomOwnerChanged
  (params (OwningPlayer :type Fort%20Player%20Controller%20Object%20Reference)))

(event Custom|CameraChanged
  (params (CameraTag :type Gameplay%20Tag%20Structure))
  (if (GameplayTags|Equal(GameplayTag) CameraTag "(TagName=\"ERG.Camera.MainRoom.Main\")")
    (CallFunction|RoomOwnerChanged (Variables|Default|GetRoomOwner))))

(fn SetHenchmenState
  (params (Owner :type Fort%20Player%20Controller%20Object%20Reference))
  (bind _owner (Variables|GetOwner))
  (bind _returnvalue (Quests|GetQuestManager _owner "Invalid"))
  (LocalVariables|SetA_B_Quest (Quests|GetQuestwithDefinition _returnvalue 0))
  (LocalVariables|SetEgoQuest (Quests|GetQuestwithDefinition _returnvalue 0))
  (LocalVariables|SetAlterQuest (Quests|GetQuestwithDefinition _returnvalue 0))
  (Utilities|IsValid (LocalVariables|GetA_B_Quest)
    (:Is%20Valid
      (if (Profile|GetItemDefinitionOwnershipState Owner 0)
        (if true
          (Utilities|IsValid (LocalVariables|GetEgoQuest)
            (:Is%20Valid
              (if (or (Quest|HasCompletedQuest (LocalVariables|GetAlterQuest) :!class /Script/FortniteGame.FortQuestItem) (Quest|HasCompletedQuest (LocalVariables|GetEgoQuest) :!class /Script/FortniteGame.FortQuestItem))
                (return 3)
                (elif (and (Utilities|Enum|NotEqual(Enum) (Quest|GetFortQuestState (LocalVariables|GetEgoQuest)) "Active") (Utilities|Enum|NotEqual(Enum) (Quest|GetFortQuestState (LocalVariables|GetAlterQuest)) "Active"))
                  (return 5)
                  (else
                    (return 1)))))
            (:Is%20Not%20Valid
              (return 1)))
          (else
            (return 5)))
        (else
          (return 0))))
    (:Is%20Not%20Valid
      (return 2))))

(fn SetEntry-0
  (bind _slot_0_state (Variables|Default|GetSlot-0-State))
  (Rendering|Material|SetVectorParameterValue (Variables|Default|GetMID-Slot0) "RGBA-1" (Utilities|Array|Get(aref) (Variables|Default|GetState-ColorArray) _slot_0_state) :!class /Script/Engine.MaterialInstanceDynamic)
  (Rendering|Material|SetScalarParameterValue (Variables|Default|GetMID-Slot0) "bIsBang" (select (== _slot_0_state 5) 1.0 0.0) :!class /Script/Engine.MaterialInstanceDynamic))

(fn SetEntry-1
  (bind _slot_1_state (Variables|Default|GetSlot-1-State))
  (Rendering|Material|SetVectorParameterValue (Variables|Default|GetMID-Slot1) "RGBA-1" (Utilities|Array|Get(aref) (Variables|Default|GetState-ColorArray) _slot_1_state) :!class /Script/Engine.MaterialInstanceDynamic)
  (Rendering|Material|SetScalarParameterValue (Variables|Default|GetMID-Slot1) "bIsBang" (select (== _slot_1_state 5) 1.0 0.0) :!class /Script/Engine.MaterialInstanceDynamic))

(fn SetEntry-2
  (bind _slot_2_state (Variables|Default|GetSlot-2-State))
  (Rendering|Material|SetVectorParameterValue (Variables|Default|GetMID-Slot2) "RGBA-1" (Utilities|Array|Get(aref) (Variables|Default|GetState-ColorArray) _slot_2_state) :!class /Script/Engine.MaterialInstanceDynamic)
  (Rendering|Material|SetScalarParameterValue (Variables|Default|GetMID-Slot2) "bIsBang" (select (== _slot_2_state 5) 1.0 0.0) :!class /Script/Engine.MaterialInstanceDynamic))

(fn SetEntry-3
  (bind _slot_3_state (Variables|Default|GetSlot-3-State))
  (Rendering|Material|SetVectorParameterValue (Variables|Default|GetMID-Slot3) "RGBA-1" (Utilities|Array|Get(aref) (Variables|Default|GetState-ColorArray) _slot_3_state) :!class /Script/Engine.MaterialInstanceDynamic)
  (Rendering|Material|SetScalarParameterValue (Variables|Default|GetMID-Slot3) "bIsBang" (select (== _slot_3_state 5) 1.0 0.0) :!class /Script/Engine.MaterialInstanceDynamic))

(fn SetEntry-4
  (bind _slot_4_state (Variables|Default|GetSlot-4-State))
  (Rendering|Material|SetVectorParameterValue (Variables|Default|GetMID-Slot4) "RGBA-1" (Utilities|Array|Get(aref) (Variables|Default|GetState-ColorArray) _slot_4_state) :!class /Script/Engine.MaterialInstanceDynamic)
  (Rendering|Material|SetScalarParameterValue (Variables|Default|GetMID-Slot4) "bIsBang" (select (== _slot_4_state 5) 1.0 0.0) :!class /Script/Engine.MaterialInstanceDynamic))

(fn ConstructionScript
  (bind _returnvalue (Rendering|Material|CreateDynamicMaterialInstance (Variables|Default|GetElevator-FloorEntry-0) 0 0 "None" :!class /Script/Engine.PrimitiveComponent))
  (Variables|Default|SetMID-Slot0 _returnvalue)
  (bind _returnvalue_2 (Rendering|Material|CreateDynamicMaterialInstance (Variables|Default|GetElevator-FloorEntry-1) 0 0 "None" :!class /Script/Engine.PrimitiveComponent))
  (Variables|Default|SetMID-Slot1 _returnvalue_2)
  (bind _returnvalue_3 (Rendering|Material|CreateDynamicMaterialInstance (Variables|Default|GetElevator-FloorEntry-2) 0 0 "None" :!class /Script/Engine.PrimitiveComponent))
  (Variables|Default|SetMID-Slot2 _returnvalue_3)
  (bind _returnvalue_4 (Rendering|Material|CreateDynamicMaterialInstance (Variables|Default|GetElevator-FloorEntry-3) 0 0 "None" :!class /Script/Engine.PrimitiveComponent))
  (Variables|Default|SetMID-Slot3 _returnvalue_4)
  (bind _returnvalue_5 (Rendering|Material|CreateDynamicMaterialInstance (Variables|Default|GetElevator-FloorEntry-4) 0 0 "None" :!class /Script/Engine.PrimitiveComponent))
  (Variables|Default|SetMID-Slot4 _returnvalue_5))
